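# Niagara swing trail for the GrimeHold pickaxe, combining two ribbon trails with toon-shaded mesh droplets.
# Both ribbons use anim-trail modules, flip twist by User.SwingSectionIndex, fade quickly and dim by time of day.
# Droplets spawn around a sampled socket in a delayed burst window, fly outward at partition-picked speeds, then fall and shrink.
# User.VariantPicker selects droplet color, and droplet gravity is reduced in the front end.
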NiagaraSystem /BRCosmetics/Weapons/FORT_Melee/Pickaxe_GrimeHold/FX/NS_Pickaxe_GrimeHold_Trail
dependencies {
 usedDataInterfaces [
  &NiagaraDataInterfaceArrayFloat
  &NiagaraDataInterfaceCurve
  &NiagaraDataInterfacePlatformSet
  &NiagaraDataInterfaceVectorCurve
  &NiagaraDataInterfaceSocketReader
 ]
 usedDynamicInputs []
 usedModules [
  &self:SystemSpawnScript.SystemScriptSource.SystemScriptGraph.NiagaraNodeAssignment_1.SetVariables_#1
  &SystemState
  &self:SystemSpawnScript.SystemScriptSource.SystemScriptGraph.NiagaraNodeAssignment_0.SetVariables_#2
 ]
}
emitters {
 NE_Droplets {
  emitterSpawnScript [
   /Game/Effects/Niagara/MasterMaterials/ToonShader/Template/NM_OutlineParameters()
   /Game/Effects/Niagara/MasterMaterials/ToonShader/Template/NM_ToonShaderParameters(HighLight_Size=0.15, HighLight_Softness=1)
  ]
  emitterUpdateScript [
   &EmitterState(Life Cycle Mode=::NewEnumerator1, Loop Behavior=::NewEnumerator1, Loop Delay=0.05, Loop Duration=0.17, UseLoopDelay=true)
   &SpawnRate(SpawnRate=50, Use Spawn Probability=false)
  ]
  particleSpawnScript [
   &PartitionParticles()
   &InitializeParticle(
    Lifetime Max=0.75
    Lifetime Min=0.55
    Lifetime Mode=::NewEnumerator1
    Mesh Scale Max=<0.015 0.015 0.015>
    Mesh Scale Min=<0.005 0.005 0.005>
    Mesh Scale Mode=::NewEnumerator2
    Sprite Rotation Mode=::NewEnumerator2
    UsePositionOffset=false
    Write Material Random=true
   )
   &SocketLocation(
    Inherited Velocity Speed Limit=300000
    SamplePosOffset=true
    Sampled Velocity Coordinate Space:ENiagaraCoordinateSpace=::Simulation
    Socket Reader=$System.Socket Reader
    Socket Sampling Mode=::NewEnumerator3
    VelocitySpeedLimit=false
   )
   &ShapeLocation(MaskSpawnGroup=false, Override Seed=false, Sphere Radius=14, Sphere Surface Distribution=0.5238, UseOwnerScale=true)
   &AddVelocity(
    Module.Use Falloff Distance=false
    Override Seed=false
    Velocity Mode=::NewEnumerator1
    Velocity Origin=$Output.SocketLocation.SampledPosition
    Velocity Speed=&SelectFloatFromArray(Direct Array Index=$Particles.PartitionParticles.Partition, Float Selection Array=DataInterface() {propertyValues {FloatData [20,150]}})
   )
  ]
  particleUpdateScript [
   &ParticleState()
   &ScaleMeshSize(Scale Factor=&VectorFromCurve(VectorCurve=DataInterface() {propertyValues {XCurve {keys [<at 0 1>,<at 1 0>]}, YCurve {keys [<at 0 1>,<at 1 0>]}, ZCurve {keys [<at 0 1>,<at 1 0>]}}}))
   &UpdateMeshOrientation()
   &Drag(Drag=&RandomRangeFloat(Minimum=3, Maximum=5), Use Linear Drag=true, Use Rotational Drag=true)
   &GravityForce(Gravity=&Multiply_VectorByFloat(
     Vector=&RandomRangeVector(Minimum=<0 0 -650>, Maximum=<0 0 -980>)
     Float=&FN_FrontEndFloatModifier(Gameplay=1, Front End=0.66)
    ))
   &SolveForcesAndVelocity(Clamp Velocity=false, Limit Acceleration=false)
   &ScaleColor(
    Scale Mode=::NewEnumerator1
    Scale RGBA=&MakeVector4_FromLinearColor(LinearColor=&FNTimeOfDayColor(
      Night Color=&MergeHelper_ColorSwitch(
       0=<rgba 0.4 0.8823 1 1>
       1=<rgba 0.0696 1 0 1>
       2=?
       3=?
       Index=&MakeIntFromFloat(Float=$User.VariantPicker)
       Num=::NewEnumerator0
      )
      Morning Color=&MergeHelper_ColorSwitch(
       0=<rgba 0.4 0.8823 1 1>
       1=<rgba 0.0696 1 0 1>
       2=?
       3=?
       Index=&MakeIntFromFloat(Float=$User.VariantPicker)
       Num=::NewEnumerator0
      )
      Day Color=&MergeHelper_ColorSwitch(
       0=<rgba 0.4 0.8823 1 1>
       1=<rgba 0.0696 1 0 1>
       2=?
       3=?
       Index=&MakeIntFromFloat(Float=$User.VariantPicker)
       Num=::NewEnumerator0
      )
      Evening Color=&MergeHelper_ColorSwitch(
       0=<rgba 0.4 0.8823 1 1>
       1=<rgba 0.0696 1 0 1>
       2=?
       3=?
       Index=&MakeIntFromFloat(Float=$User.VariantPicker)
       Num=::NewEnumerator0
      )
     ))
    ScaleRGBA=true
   ) {module=ScaleColor001}
  ]
  rendererClasses [&NiagaraMeshRendererProperties]
  renderers [{rendererClass=&NiagaraMeshRendererProperties, rendererIndex=0}]
 }
 NE_Swing_01 {
  emitterSpawnScript [&FN_RibbonIDOffsetUtility()]
  emitterUpdateScript [
   &EmitterState(Life Cycle Mode=::NewEnumerator1, Loop Behavior=::NewEnumerator1, Loop Duration=0.35, UseLoopDelay=false)
   &FN_SpawnAnimTrail()
  ]
  particleSpawnScript [
   &FN_InitializeAnimTrailBoneIndecies(First Bone or Socket Index=2)
   &FN_InitializeAnimTrail(AnimTrailWidth=0.75, Color=<rgba 0 0 0 2>, Lifetime=0.175)
   &AddVelocity(Velocity=<0 400 0>) {enabled=false}
  ]
  particleUpdateScript [
   &CameraOffset()
   /SetVariables(Particles.RibbonTwist:NiagaraFloat=&SelectFloatFromArray(Direct Array Index=$User.SwingSectionIndex, Float Selection Array=DataInterface() {propertyValues {FloatData [0,180,0,180]}})) {isSetParameters=true}
   &SolveForcesAndVelocity(Clamp Velocity=false, Limit Acceleration=false)
   &FN_UpdateAnimTrail() {module=FN_UpdateAnimTrail001}
   &ScaleColor(
    Scale Alpha=&FloatFromCurve(FloatCurve=DataInterface() {propertyValues {Curve {keys [
         {arriveTangent=-6.8242, interpMode=RCIM_Cubic, leaveTangent=-6.8242, tangentMode=RCTM_User, time=0, value=1}
         {interpMode=RCIM_Cubic, tangentMode=RCTM_User, time=0.4, value=0}
        ]}}})
    Scale RGB=&VectorFromFloat(Value=&FNTimeOfDayScalar(Night Scale=0.15, Morning Scale=0.5, Evening Scale=0.5))
    ScaleA=true
    ScaleRGB=true
   )
  ]
  rendererClasses [&NiagaraRibbonRendererProperties]
  renderers [{rendererClass=&NiagaraRibbonRendererProperties}]
 }
 NE_Swing_02 {
  emitterSpawnScript [&FN_RibbonIDOffsetUtility()]
  emitterUpdateScript [
   &EmitterState(Life Cycle Mode=::NewEnumerator1, Loop Behavior=::NewEnumerator1, Loop Duration=0.35, UseLoopDelay=false)
   &FN_SpawnAnimTrail()
  ]
  particleSpawnScript [
   &FN_InitializeAnimTrailBoneIndecies(First Bone or Socket Index=2)
   &FN_InitializeAnimTrail(AnimTrailWidth=0.7, Color=<rgba 0 0 0 2>, Lifetime=0.175)
   &AddVelocity(Velocity=<0 400 0>) {enabled=false}
  ]
  particleUpdateScript [
   &CameraOffset()
   /SetVariables(Particles.RibbonTwist:NiagaraFloat=&SelectFloatFromArray(Direct Array Index=$User.SwingSectionIndex, Float Selection Array=DataInterface() {propertyValues {FloatData [0,180,0,180]}})) {isSetParameters=true}
   &SolveForcesAndVelocity(Clamp Velocity=false, Limit Acceleration=false)
   &FN_UpdateAnimTrail() {module=FN_UpdateAnimTrail001}
   &ScaleColor(
    Scale Alpha=&FloatFromCurve(FloatCurve=DataInterface() {propertyValues {Curve {keys [
         {arriveTangent=-6.8242, interpMode=RCIM_Cubic, leaveTangent=-6.8242, tangentMode=RCTM_User, time=0, value=1}
         {interpMode=RCIM_Cubic, tangentMode=RCTM_User, time=0.4, value=0}
        ]}}})
    Scale RGB=&VectorFromFloat(Value=&FNTimeOfDayScalar(Night Scale=0.15, Morning Scale=0.5, Evening Scale=0.5))
    ScaleA=true
    ScaleRGB=true
   )
  ]
  rendererClasses [&NiagaraRibbonRendererProperties]
  renderers [{rendererClass=&NiagaraRibbonRendererProperties}]
 }
}
summary {
 systemName=NS_Pickaxe_GrimeHold_Trail
 emitters [
  {emitterName=NE_Swing_01, rendererClasses [&NiagaraRibbonRendererProperties]}
  {emitterName=NE_Droplets, rendererClasses [&NiagaraMeshRendererProperties]}
  {emitterName=NE_Swing_02, rendererClasses [&NiagaraRibbonRendererProperties]}
 ]
 userVariables [
  User.bIsFrontEnd {type=NiagaraBool}
  User.bIsFrontEndPreview {type=NiagaraBool}
  User.PhysicalSurface {type=NiagaraInt32}
  User.RibbonIDOffset {type=NiagaraInt32}
  User.SwingSectionIndex {type=NiagaraInt32}
  User.VariantPicker {type=NiagaraFloat, defaultValue=0, description=""}
 ]
}
