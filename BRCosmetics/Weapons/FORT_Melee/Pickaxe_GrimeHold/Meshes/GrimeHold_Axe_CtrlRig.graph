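# Control Rig for the GrimeHold fish pickaxe that turns one Blink variable into a staggered blink across right, middle and left blink pose curves.
# Blink is interpolated into R_blink_pose, which is then read back through short time-offset buffers to drive M_blink_pose and, later, L_blink_pose.
# The look curves and bone hierarchy are not driven here, and an L_blink_offset setter sits unconnected to execution.
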ControlRigBlueprint /BRCosmetics/Weapons/FORT_Melee/Pickaxe_GrimeHold/Meshes/GrimeHold_Axe_CtrlRig
graphs [
 RigVMFunctionLibrary {links [], nodes []}
 RigVMModel {nodes [
   RigUnit_BeginExecution = &RigUnit_BeginExecution()
   n1 = &RigVMFunction_AlphaInterp(
    InterpSpeedDecreasing=25
    InterpSpeedIncreasing=25
    Scale=1
    bClampResult=False
    bMapRange=false
    Value=RigVMVariableNode() {variable=Blink}.Value
   )
   /Script/ControlRig.RigUnit_SetCurveValue(Curve=R_blink_pose, Value=@n1.Result, ExecutePin=@RigUnit_BeginExecution.ExecutePin) {is=struct RigVMUnitNode}
   AlphaInterp = &RigVMFunction_AlphaInterp(
    InterpSpeedDecreasing=25
    InterpSpeedIncreasing=25
    Scale=1
    bClampResult=False
    bInterpResult=false
    bMapRange=false
    Value=/Script/RigVM.RigVMFunction_TimeOffsetFloat(
     BufferSize=16
     SecondsAgo=0.1
     TimeRange=0.8
     Value=/Script/ControlRig.RigUnit_GetCurveValue(Curve=R_blink_pose) {is=struct RigVMUnitNode}.Value
    ) {is=struct RigVMUnitNode, template="TimeOffset::Execute(in Value,in SecondsAgo,in BufferSize,in TimeRange,out Result)"}.Result
   )
   RigVMFunction_AlphaInterp = &RigVMFunction_AlphaInterp(InterpSpeedDecreasing=10, InterpSpeedIncreasing=10, Scale=1, bClampResult=False, bInterpResult=False, Value=@AlphaInterp.Result)
   > &RigUnit_SetCurveValue(Curve=M_blink_pose, Value=@RigVMFunction_AlphaInterp.Result)
   n2 = &RigVMFunction_AlphaInterp(
    InterpSpeedDecreasing=25
    InterpSpeedIncreasing=25
    Scale=1
    bClampResult=False
    bInterpResult=false
    bMapRange=false
    Value=&RigVMFunction_TimeOffsetFloat(BufferSize=16, SecondsAgo=0.25, TimeRange=0.8, Value=&RigUnit_GetCurveValue(Curve=R_blink_pose).Value) {template="TimeOffset::Execute(in Value,in SecondsAgo,in BufferSize,in TimeRange,out Result)"}.Result
   )
   > &RigUnit_SetCurveValue(Curve=L_blink_pose, Value=@n2.Result)
   &RigUnit_SetCurveValue(Curve=L_blink_offset, ExecutePin=(), Value=1.1)
  ]}
]
hierarchy [
 L_blink_pose {type=Curve, loc=<0 0 0>, rot=<0 0 0 1>}
 M_blink_pose {type=Curve, loc=<0 0 0>, rot=<0 0 0 1>}
 R_blink_pose {type=Curve, loc=<0 0 0>, rot=<0 0 0 1>}
 look_right_pose {type=Curve, loc=<0 0 0>, rot=<0 0 0 1>}
 look_left_pose {type=Curve, loc=<0 0 0>, rot=<0 0 0 1>}
 look_up_pose {type=Curve, loc=<0 0 0>, rot=<0 0 0 1>}
 look_down_pose {type=Curve, loc=<0 0 0>, rot=<0 0 0 1>}
 b_AR_Root {loc=<0 0 0>, rot=<0 0 0 1>}
 Melee_TwoHanded_Handle {parent=b_AR_Root, loc=<0 0 0>, rot=<0 0 0.7071 -0.7071>}
 base_head_cr {parent=Melee_TwoHanded_Handle, loc=<-10.7949 0.7899 31.2683>, rot=<0 0 -0.7071 -0.7071>}
 dyn_head_cr {parent=base_head_cr, loc=<0 0 0>, rot=<0 0 0 1>}
 base_fish_00_cr {parent=dyn_head_cr, loc=<-1.3427 -10.0813 -31.3453>, rot=<-0.5629 0.4425 -0.4271 -0.5522>}
 dyn_fish_00_cr {parent=base_fish_00_cr, loc=<0 0 0>, rot=<0 0 0 1>}
 dyn_fish_01_cr {parent=dyn_fish_00_cr, loc=<6.4446 0.1212 0.0923>, rot=<0.0102 -0.1146 -0.0884 -0.9894>}
 dyn_fish_02_cr {parent=dyn_fish_01_cr, loc=<9.1311 -0.0495 0.0298>, rot=<0.0003 -0.0961 0.0111 -0.9953>}
 dyn_fish_03_cr {parent=dyn_fish_02_cr, loc=<7.8637 1.4183 -1.032>, rot=<0.0169 -0.0321 -0.0378 -0.9986>}
 dyn_fish_04_cr {parent=dyn_fish_03_cr, loc=<7.6631 0 -0.2932>, rot=<0 0.0664 0 -0.9978>}
 base_fin_01_rt {parent=dyn_fish_04_cr, loc=<3.5914 17.8554 -4.994>, rot=<-0.2614 -0.6053 0.7169 0.2265>}
 dyn_fin_01_rt {parent=base_fin_01_rt, loc=<0 0 0>, rot=<0 0 0 1>}
 dyn_fin_02_rt {parent=dyn_fin_01_rt, loc=<5.2226 0 -0.8338>, rot=<0 0 0 1>}
 base_fin_01_lf {parent=dyn_fish_04_cr, loc=<3.1653 -17.0781 -5.5506>, rot=<0.2629 -0.7515 0.5356 -0.2816>}
 dyn_fin_01_lf {parent=base_fin_01_lf, loc=<0 0 0>, rot=<0 0 0 1>}
 dyn_fin_02_lf {parent=dyn_fin_01_lf, loc=<4.2858 0 0.6252>, rot=<0 0 0 1>}
 base_head_fin_center_01_cr {parent=dyn_fish_04_cr, loc=<20.0067 0.2839 6.7689>, rot=<-0.0054 0.0412 0.0019 -0.9991>}
 dyn_head_fin_center_01_cr {parent=base_head_fin_center_01_cr, loc=<0 0 0>, rot=<0 0 0 1>}
 dyn_head_fin_center_02_cr {parent=dyn_head_fin_center_01_cr, loc=<3.815 0 0>, rot=<0 0 0 1>}
 base_head_fin_front_01_cr {parent=dyn_fish_04_cr, loc=<21.2989 0.322 3.9541>, rot=<-0.005 0.0143 -0.0042 -0.9999>}
 dyn_head_fin_front_01_cr {parent=base_head_fin_front_01_cr, loc=<0 0 0>, rot=<0 0 0 1>}
 dyn_head_fin_front_02_cr {parent=dyn_head_fin_front_01_cr, loc=<4.3216 0 0>, rot=<0 0 0 1>}
 base_head_fin_back_01_cr {parent=dyn_fish_04_cr, loc=<18.5802 0.2182 9.3478>, rot=<-0.0046 0.0545 -0.0147 -0.9984>}
 dyn_head_fin_back_01_cr {parent=base_head_fin_back_01_cr, loc=<0 0 0>, rot=<0 0 0 1>}
 dyn_head_fin_back_02_cr {parent=dyn_head_fin_back_01_cr, loc=<3.1419 0 -0.4353>, rot=<0 0 0 1>}
 FX_Swing_02 {parent=dyn_head_fin_back_02_cr, type=Null, loc=<0 0 0>, rot=<0 0 0 1>}
 base_mouth_fish_cr {parent=dyn_fish_04_cr, loc=<8.3277 -0.3719 -0.753>, rot=<-0.2579 -0.1059 -0.0494 -0.9591>}
 dyn_jaw_cr {parent=base_mouth_fish_cr, loc=<-6.0992 1.0144 -0.9409>, rot=<0.4626 -0.2068 0.7253 -0.466>}
 base_lip_lower_cr {parent=dyn_jaw_cr, loc=<0.0507 17.8969 7.6645>, rot=<-0.2586 0.2586 -0.6581 -0.6581>}
 dyn_lip_lower_cr {parent=base_lip_lower_cr, loc=<0 0 0>, rot=<0 0 0 1>}
 FX_Swing_01 {parent=dyn_lip_lower_cr, type=Null, loc=<0 0 0>, rot=<0 0 0 1>}
 base_lip_upper_cr {parent=base_mouth_fish_cr, loc=<10.1678 -5.9765 -8.8333>, rot=<0.2034 -0.3586 0.158 -0.8973>}
 dyn_lip_upper_cr {parent=base_lip_upper_cr, loc=<0 0 0>, rot=<0 0 0 1>}
 base_lip_corner_rt {parent=base_mouth_fish_cr, loc=<9.9808 -0.659 -13.4069>, rot=<0.2264 -0.2076 0.1228 -0.9437>}
 dyn_lip_corner_rt {parent=base_lip_corner_rt, loc=<0 0 0>, rot=<0 0 0 1>}
 base_lip_corner_lf {parent=base_mouth_fish_cr, loc=<8.1154 -12.4082 -6.9291>, rot=<0.2264 -0.2076 0.1228 -0.9437>}
 dyn_lip_corner_lf {parent=base_lip_corner_lf, loc=<0 0 0>, rot=<0 0 0 1>}
 body_root {parent=b_AR_Root, loc=<0 0 0>, rot=<0 0 0 1>}
 pelvis {parent=body_root, loc=<0 -0.7071 93.1191>, rot=<0.0127 0.707 -0.0127 -0.707>}
 spine_01 {parent=pelvis, loc=<6.7494 -0.6388 0>, rot=<0 0 0.034 -0.9994>}
 spine_02 {parent=spine_01, loc=<7.2518 -0.035 0>, rot=<0 0 0.0238 -0.9997>}
 spine_03 {parent=spine_02, loc=<6.0104 0 0>, rot=<0 0 0.0149 -0.9999>}
 spine_04 {parent=spine_03, loc=<8.1024 0 0>, rot=<0 0 -0.0132 -0.9999>}
 thigh_l {parent=pelvis, loc=<-2.5693 -0.0109 -7.7501>, rot=<0.0012 0.0572 0.0155 -0.9982>}
 calf_l {parent=thigh_l, loc=<-39.0484 0 0>, rot=<0.0006 -0.0161 -0.0391 -0.9991>}
 thigh_r {parent=pelvis, loc=<-2.5693 -0.0109 7.7501>, rot=<0.0572 -0.0012 0.9982 0.0155>}
 calf_r {parent=thigh_r, loc=<39.0484 0 0>, rot=<0.0006 -0.0161 -0.0391 -0.9991>}
 Char_Thigh_r {parent=b_AR_Root, loc=<21 0 -0.0024>, rot=<0 0 0 1>}
]
variables [Blink {type=double, default=0}]
