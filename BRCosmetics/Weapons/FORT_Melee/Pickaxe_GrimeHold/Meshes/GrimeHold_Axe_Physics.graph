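# Physics asset for the GrimeHold pickaxe mesh, defining capsule bodies for a jointed fish chain and for thigh and calf bones.
# The fish chain runs from a small base through five dyn bodies, swinging freely at the base, with swing limits tightening toward the larger end capsules and zero twist limits.
# Each calf is locked to its thigh, and no constraint links the legs to the fish.
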PhysicsAsset /BRCosmetics/Weapons/FORT_Melee/Pickaxe_GrimeHold/Meshes/GrimeHold_Axe_Physics
bodies {
 base_fish_00_cr [{shapeType=Capsule, length=0.505, radius=0.505, shapeName=base_fish_00_cr_capsule, rotation=<rot 0 -90 0>}]
 calf_l [{shapeType=Capsule, length=28, radius=11, shapeName=thigh_l_capsule, center=<-26 -3.9752 -3.1116>, rotation=<rot 90 180 0>}]
 calf_r [{shapeType=Capsule, length=28, radius=11, shapeName=thigh_r_capsule, center=<26 6.0166 3.1116>, rotation=<rot 90 0 0>}]
 dyn_fish_00_cr [{shapeType=Capsule, length=2, radius=5, shapeName=dyn_fish_00_cr_capsule, center=<5 0 0>, rotation=<rot 90 0 0>}]
 dyn_fish_01_cr [{shapeType=Capsule, length=5, radius=5, shapeName=dyn_fish_01_cr_capsule, center=<5 0 0>, rotation=<rot 0 90 0>}]
 dyn_fish_02_cr [{shapeType=Capsule, length=5, radius=5, shapeName=dyn_fish_02_cr_capsule, center=<5 0 0>, rotation=<rot 0 90 0>}]
 dyn_fish_03_cr [{shapeType=Capsule, length=0, radius=20, shapeName=dyn_fish_03_cr_capsule, center=<10 0 -1.8>, rotation=<rot 0 90 0>}]
 dyn_fish_04_cr [{shapeType=Capsule, length=1, radius=18.5, shapeName=dyn_fish_04_cr_capsule, center=<10 0 -1.5>, rotation=<rot 0 90 0>}]
 thigh_l [{shapeType=Capsule, length=32, radius=12, shapeName=thigh_l_capsule, center=<-22.0878 -0.4671 -3.3572>, rotation=<rot 90 -170.103 0>}]
 thigh_r [{shapeType=Capsule, length=32, radius=12, shapeName=thigh_r_capsule, center=<21.8785 1.6664 3.3572>, rotation=<rot 90 9.897 0>}]
}
constraints [
 {
  bone1Name=calf_l
  bone2Name=thigh_l
  swing1LimitDegrees=45
  swing1Motion=Locked
  swing2LimitDegrees=45
  swing2Motion=Locked
  twistLimitDegrees=45
  twistMotion=Locked
 }
 {
  bone1Name=calf_r
  bone2Name=thigh_r
  swing1LimitDegrees=45
  swing1Motion=Locked
  swing2LimitDegrees=45
  swing2Motion=Locked
  twistLimitDegrees=45
  twistMotion=Locked
 }
 {bone1Name=dyn_fish_03_cr, bone2Name=dyn_fish_02_cr, swing1LimitDegrees=15, swing2LimitDegrees=15, twistLimitDegrees=0}
 {bone1Name=dyn_fish_02_cr, bone2Name=dyn_fish_01_cr, swing1LimitDegrees=20, swing2LimitDegrees=20, twistLimitDegrees=0}
 {bone1Name=dyn_fish_01_cr, bone2Name=dyn_fish_00_cr, swing1LimitDegrees=40, swing2LimitDegrees=40, twistLimitDegrees=0}
 {
  bone1Name=dyn_fish_00_cr
  bone2Name=base_fish_00_cr
  swing1LimitDegrees=45
  swing1Motion=Free
  swing2LimitDegrees=45
  swing2Motion=Free
  twistLimitDegrees=45
  twistMotion=Free
 }
 {bone1Name=dyn_fish_04_cr, bone2Name=dyn_fish_03_cr, swing1LimitDegrees=15, swing2LimitDegrees=15, twistLimitDegrees=0}
]
