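# Physics asset for the GrimeHold axe pickaxe cosmetic, simulating a dangling fish bone chain on the weapon.
# A capsule chain runs from a tiny base body through four dynamic fish bones, free to rotate at the base, with limited swing and zero twist limit further along.
# Thigh and calf capsules on both sides, with each calf rotationally locked to its thigh, appear to act as leg colliders for the fish.
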PhysicsAsset /BRCosmetics/Weapons/FORT_Melee/Pickaxe_GrimeHold/Meshes/GrimeHold_Axe_PhysicsFE
bodies {
 base_fish_00_cr [{shapeType=Capsule, length=0.505, radius=0.505, shapeName=base_fish_00_cr_capsule, rotation=<rot 0 -90 0>}]
 calf_l [{shapeType=Capsule, length=26.5, radius=13, shapeName=thigh_l_capsule, center=<-26 -1.6743 -3.1116>, rotation=<rot 90 180 0>}]
 calf_r [{shapeType=Capsule, length=26.5, radius=13, shapeName=thigh_r_capsule, center=<26 1.6743 3.1116>, rotation=<rot 90 0 0>}]
 dyn_fish_00_cr [{shapeType=Capsule, length=9.7943, radius=7.3004, shapeName=dyn_fish_00_cr_capsule, center=<3 0 0>, rotation=<rot 90 0 0>}]
 dyn_fish_01_cr [{shapeType=Capsule, length=5, radius=5, shapeName=dyn_fish_01_cr_capsule, rotation=<rot 0 90 0>}]
 dyn_fish_02_cr [{shapeType=Capsule, length=5, radius=5, shapeName=dyn_fish_02_cr_capsule, rotation=<rot 0 90 0>}]
 dyn_fish_03_cr [{shapeType=Capsule, length=1, radius=24.5, shapeName=dyn_fish_03_cr_capsule, center=<13 0 0>, rotation=<rot 0 90 0>}]
 thigh_l [{
   shapeType=Capsule
   length=30.6616
   radius=13.5479
   shapeName=thigh_l_capsule
   center=<-22.1292 -0.2295 -3.3572>
   rotation=<rot 90 -170.103 0>
  }]
 thigh_r [{shapeType=Capsule, length=30.6616, radius=13.5479, shapeName=thigh_r_capsule, center=<22.1292 0.2295 3.3572>, rotation=<rot 90 9.897 0>}]
}
constraints [
 {
  bone1Name=calf_l
  bone2Name=thigh_l
  swing1LimitDegrees=45
  swing1Motion=Locked
  swing2LimitDegrees=45
  swing2Motion=Locked
  twistLimitDegrees=45
  twistMotion=Locked
 }
 {
  bone1Name=calf_r
  bone2Name=thigh_r
  swing1LimitDegrees=45
  swing1Motion=Locked
  swing2LimitDegrees=45
  swing2Motion=Locked
  twistLimitDegrees=45
  twistMotion=Locked
 }
 {bone1Name=dyn_fish_03_cr, bone2Name=dyn_fish_02_cr, swing1LimitDegrees=45, swing2LimitDegrees=45, twistLimitDegrees=0}
 {bone1Name=dyn_fish_02_cr, bone2Name=dyn_fish_01_cr, swing1LimitDegrees=45, swing2LimitDegrees=45, twistLimitDegrees=0}
 {bone1Name=dyn_fish_01_cr, bone2Name=dyn_fish_00_cr, swing1LimitDegrees=45, swing2LimitDegrees=45, twistLimitDegrees=0}
 {
  bone1Name=dyn_fish_00_cr
  bone2Name=base_fish_00_cr
  swing1LimitDegrees=45
  swing1Motion=Free
  swing2LimitDegrees=45
  swing2Motion=Free
  twistLimitDegrees=45
  twistMotion=Free
 }
]
