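# Animation Blueprint for the GrimeHold pickaxe, animating its fish-like head with physics and random idle blinks, fin flaps and mouth opening.
# Body-bone overrides, fin/mouth constraints and jaw offsets feed two RigidBody simulations, then fin flaps, spring AnimDynamics and a blink Control Rig.
# Each update caches parent-mesh socket transforms, sets gameplay versus front-end RigidBody alphas, swaps sim-space settings during melee and runs randomized Delay timers for idle motions.
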AnimBlueprint /BRCosmetics/Weapons/FORT_Melee/Pickaxe_GrimeHold/Meshes/GrimeHold_Axe_AnimBP
target_skeleton=&/GrimeHold_Axe_Skeleton
graphs [
 AnimGraph {nodes [
   n1 = Control Rig!(Source=Use cached pose 'fin_cache'(), Alpha, Blink:double=Get Blink() -> (Blink:double)) -> (Pose) {controlRigAssetReference {blueprintRigClass=&/GrimeHold_Axe_CtrlRig.GrimeHold_Axe_CtrlRig_C}}
   Output Pose(Result=@n1) {layerGroup=None}
   AnimGraphNode_SubInput_0 = Input Pose()
   n2 = "Transform (Modify) Bone - Bone: body_root"(ComponentPose=Local To Component(LocalPose=@AnimGraphNode_SubInput_0), Alpha=1, Translation="0,0,0", Rotation="0,0,0", Scale="1,1,1") {
    rotationMode=BMM_Replace
    rotationSpace=BCS_WorldSpace
    scaleSpace=BCS_BoneSpace
    translationMode=BMM_Replace
    translationSpace=BCS_WorldSpace
    alphaBoolBlend {blendOption=Linear}
    alphaScaleBiasClamp {bInterpResult=true, interpSpeedDecreasing=1, interpSpeedIncreasing=1, inRange {max=0, min=-40}, outRange {max=0, min=1}}
    boneToModify {boneName=body_root}
   }
   n3 = "Transform (Modify) Bone - Bone: pelvis"(ComponentPose=@n2, Alpha=1, Translation="0,0,0", Rotation="0,0,0", Scale="1,1,1") {
    rotationMode=BMM_Replace
    rotationSpace=BCS_WorldSpace
    scaleSpace=BCS_BoneSpace
    translationMode=BMM_Replace
    translationSpace=BCS_WorldSpace
    alphaBoolBlend {blendOption=Linear}
    alphaScaleBiasClamp {inRange {max=0, min=-40}, outRange {max=0, min=1}}
    boneToModify {boneName=pelvis}
   }
   n4 = "Transform (Modify) Bone - Bone: thigh_l"(ComponentPose=@n3, Alpha=1, Translation="0,0,0", Rotation="0,0,0", Scale="1,1,1") {
    rotationMode=BMM_Replace
    rotationSpace=BCS_WorldSpace
    scaleSpace=BCS_BoneSpace
    translationMode=BMM_Replace
    translationSpace=BCS_WorldSpace
    alphaBoolBlend {blendOption=Linear}
    alphaScaleBiasClamp {inRange {max=0, min=-40}, outRange {max=0, min=1}}
    boneToModify {boneName=thigh_l}
   }
   n5 = "Transform (Modify) Bone - Bone: thigh_r"(ComponentPose=@n4, Alpha=1, Translation="0,0,0", Rotation="0,0,0", Scale="1,1,1") {
    rotationMode=BMM_Replace
    rotationSpace=BCS_WorldSpace
    scaleSpace=BCS_BoneSpace
    translationMode=BMM_Replace
    translationSpace=BCS_WorldSpace
    alphaBoolBlend {blendOption=Linear}
    alphaScaleBiasClamp {inRange {max=0, min=-40}, outRange {max=0, min=1}}
    boneToModify {boneName=thigh_r}
   }
   n6 = "Transform (Modify) Bone - Bone: calf_l"(ComponentPose=@n5, Alpha=1, Translation="0,0,0", Rotation="0,0,0", Scale="1,1,1") {
    rotationMode=BMM_Replace
    rotationSpace=BCS_WorldSpace
    scaleSpace=BCS_BoneSpace
    translationMode=BMM_Replace
    translationSpace=BCS_WorldSpace
    alphaBoolBlend {blendOption=Linear}
    alphaScaleBiasClamp {inRange {max=0, min=-40}, outRange {max=0, min=1}}
    boneToModify {boneName=calf_l}
   }
   n7 = "Transform (Modify) Bone - Bone: calf_r"(ComponentPose=@n6, Alpha=1, Translation="0,0,0", Rotation="0,0,0", Scale="1,1,1") {
    rotationMode=BMM_Replace
    rotationSpace=BCS_WorldSpace
    scaleSpace=BCS_BoneSpace
    translationMode=BMM_Replace
    translationSpace=BCS_WorldSpace
    alphaBoolBlend {blendOption=Linear}
    alphaScaleBiasClamp {inRange {max=0, min=-40}, outRange {max=0, min=1}}
    boneToModify {boneName=calf_r}
   }
   n8 = "Constraint - dyn_head_fin_center_01_cr (P)"!(ComponentPose=@n7, Alpha=1, ConstraintWeights_0=1, ConstraintWeights_1:float=0) -> (Pose) {boneToModify {boneName=dyn_head_fin_center_01_cr}, constraintSetup [{targetBone {boneName=dyn_fish_04_cr}}], constraintWeights [1,0]}
   n9 = "Constraint - dyn_head_fin_front_01_cr (P)"!(ComponentPose=@n8, Alpha=1, ConstraintWeights_0=1, ConstraintWeights_1:float=0) -> (Pose) {boneToModify {boneName=dyn_head_fin_front_01_cr}, constraintSetup [{targetBone {boneName=dyn_fish_04_cr}}], constraintWeights [1,0]}
   n10 = "Constraint - base_head_fin_back_01_cr (P)"!(ComponentPose=@n9, Alpha=1, ConstraintWeights_0=1, ConstraintWeights_1:float=0) -> (Pose) {boneToModify {boneName=base_head_fin_back_01_cr}, constraintSetup [{targetBone {boneName=dyn_fish_04_cr}}], constraintWeights [1,0]}
   n11 = "Constraint - base_head_fin_front_01_cr (P)"!(ComponentPose=@n10, Alpha=1, ConstraintWeights_0=1, ConstraintWeights_1:float=0) -> (Pose) {boneToModify {boneName=base_head_fin_front_01_cr}, constraintSetup [{targetBone {boneName=dyn_fish_04_cr}}], constraintWeights [1,0]}
   n12 = "Constraint - base_head_fin_center_01_cr (P)"!(ComponentPose=@n11, Alpha=1, ConstraintWeights_0=1, ConstraintWeights_1:float=0) -> (Pose) {boneToModify {boneName=base_head_fin_center_01_cr}, constraintSetup [{targetBone {boneName=dyn_fish_04_cr}}], constraintWeights [1,0]}
   n13 = "Constraint - base_mouth_fish_cr (P)"!(ComponentPose=@n12, Alpha=1, ConstraintWeights_0=1, ConstraintWeights_1:float=0) -> (Pose) {boneToModify {boneName=base_mouth_fish_cr}, constraintSetup [{targetBone {boneName=dyn_fish_03_cr}}], constraintWeights [1,0]}
   n14 = Get mouth_open() -> (mouth_open:double)
   n15 = "Transform (Modify) Bone - Bone: dyn_jaw_cr"(ComponentPose=@n13, Alpha=@n14, Translation="0,-2,0", Rotation="0,0,8.5", Scale="1,1,1") {
    rotationMode=BMM_Additive
    rotationSpace=BCS_BoneSpace
    translationMode=BMM_Additive
    translationSpace=BCS_BoneSpace
    alphaBoolBlend {blendOption=Linear}
    alphaScaleBiasClamp {bInterpResult=true, interpSpeedDecreasing=15, interpSpeedIncreasing=15}
    boneToModify {boneName=dyn_jaw_cr}
    rotation {roll=8.5}
    translation {y=-2}
   }
   n16 = "Transform (Modify) Bone - Bone: dyn_lip_corner_lf"(ComponentPose=@n15, Alpha=@n14, Translation="-5,0,0", Rotation="0,0,0", Scale="1,1,1") {
    rotationMode=BMM_Additive
    rotationSpace=BCS_BoneSpace
    translationMode=BMM_Additive
    alphaBoolBlend {blendOption=Linear}
    alphaScaleBiasClamp {bInterpResult=true, interpSpeedDecreasing=15, interpSpeedIncreasing=15}
    boneToModify {boneName=dyn_lip_corner_lf}
    translation {x=-5}
   }
   n17 = "Transform (Modify) Bone - Bone: dyn_lip_corner_rt"(ComponentPose=@n16, Alpha=@n14, Translation="5,0,0", Rotation="0,0,0", Scale="1,1,1") {
    rotationMode=BMM_Additive
    rotationSpace=BCS_BoneSpace
    translationMode=BMM_Additive
    alphaBoolBlend {blendOption=Linear}
    alphaScaleBiasClamp {bInterpResult=true, interpSpeedDecreasing=15, interpSpeedIncreasing=15}
    boneToModify {boneName=dyn_lip_corner_rt}
    translation {x=5}
   }
   n18 = RigidBody!(
    ComponentPose=@n17
    Alpha
    SimSpaceSettings:Sim Space Settings Structure=(WorldAlpha=0.8,VelocityScaleZ=0.5,DampingAlpha=0,MaxLinearVelocity=700,MaxAngularVelocity=8,MaxLinearAcceleration=3500,MaxAngularAcceleration=45,ExternalLinearDragV=(X=0.5,Y=0.5,Z=0.35),ExternalLinearVelocity=(X=0,Y=0,Z=0),ExternalAngularVelocity=(X=0,Y=0,Z=0))
   ) -> (Pose) {
    bClampLinearTranslationLimitToRefPose=true
    bOverrideWorldGravity=true
    overridePhysicsAsset=&/GrimeHold_Axe_Physics
    simulationSpace=BaseBoneSpace
    alphaBoolBlend {blendOption=Linear}
    baseBoneRef {boneName=b_AR_Root}
    overrideWorldGravity {z=-980}
    simSpaceSettings {
     dampingAlpha=0
     maxAngularAcceleration=45
     maxAngularVelocity=8
     maxLinearAcceleration=3500
     maxLinearVelocity=700
     velocityScaleZ=0.5
     worldAlpha=0.8
     externalLinearDragV=<0.5 0.5 0.35>
    }
   }
   n19 = RigidBody!(ComponentPose=@n18, Alpha) -> (Pose) {
    bClampLinearTranslationLimitToRefPose=true
    bOverrideWorldGravity=true
    overridePhysicsAsset=&/GrimeHold_Axe_PhysicsFE
    simulationSpace=BaseBoneSpace
    alphaBoolBlend {blendOption=Linear}
    baseBoneRef {boneName=b_AR_Root}
    overrideWorldGravity {z=-980}
    simSpaceSettings {maxAngularAcceleration=700, maxAngularVelocity=700, maxLinearAcceleration=700, maxLinearVelocity=700, velocityScaleZ=5, worldAlpha=0.5}
   }
   Save cached pose 'rigid_cache'(Pose=Component To Local(ComponentPose=@n19)) {cacheName=rigid_cache}
   n20 = Anim Dynamics - Bone: dyn_head_fin_front_01_cr!(ComponentPose=Local To Component(LocalPose=Use cached pose 'modifytransform_cache'()), Alpha=1, GravityOverride:Vector="0,0,-980") -> (Pose) {
    angularDampingOverride=1
    angularSpringConstant=10000
    bAngularSpring=true
    bOverrideAngularDamping=true
    bUseGravityOverride=true
    gravityScale=0.35
    simulationSpace=RootRelative
    alphaBoolBlend {blendOption=Linear}
    boundBone {boneName=dyn_head_fin_front_01_cr}
    componentAppliedLinearAccClamp=<3000 3000 3000>
    componentLinearAccScale=<0.1 0.1 0.1>
    componentLinearVelScale=<0.1 0.1 0.1>
    gravityOverride {z=-980}
    physicsBodyDefinitions [{
      boundBone {boneName=dyn_head_fin_front_01_cr}
      boxExtents=<14.5 24.5 24.5>
      constraintSetup {angularLimitsMax=<15 15 15>, angularLimitsMin=<-15 -15 -15>, angularTarget=<1 0 0>}
      localJointOffset=<0.8 0 0>
     }]
    simSpaceSettings {maxAngularAcceleration=35, maxAngularVelocity=7, simSpaceAngularAlpha=1}
   }
   n21 = Anim Dynamics - Bone: dyn_head_fin_front_02_cr!(ComponentPose=@n20, Alpha=1, GravityOverride:Vector="0,0,-980") -> (Pose) {
    angularDampingOverride=1
    angularSpringConstant=10000
    bAngularSpring=true
    bOverrideAngularDamping=true
    bUseGravityOverride=true
    gravityScale=0.35
    simulationSpace=RootRelative
    alphaBoolBlend {blendOption=Linear}
    boundBone {boneName=dyn_head_fin_front_02_cr}
    componentAppliedLinearAccClamp=<3000 3000 3000>
    componentLinearAccScale=<0.1 0.1 0.1>
    componentLinearVelScale=<0.1 0.1 0.1>
    gravityOverride {z=-980}
    physicsBodyDefinitions [{
      boundBone {boneName=dyn_head_fin_front_02_cr}
      boxExtents=<14.5 24.5 24.5>
      constraintSetup {angularLimitsMax=<15 15 8>, angularLimitsMin=<-15 -15 -8>, angularTarget=<1 0 0>}
      localJointOffset=<0.8 0 0>
     }]
    simSpaceSettings {maxAngularAcceleration=35, maxAngularVelocity=7, simSpaceAngularAlpha=1}
   }
   n22 = Anim Dynamics - Bone: dyn_head_fin_center_01_cr!(ComponentPose=@n21, Alpha=1, GravityOverride:Vector="0,0,-980") -> (Pose) {
    angularDampingOverride=1
    angularSpringConstant=10000
    bAngularSpring=true
    bOverrideAngularDamping=true
    bUseGravityOverride=true
    gravityScale=0.35
    simulationSpace=RootRelative
    alphaBoolBlend {blendOption=Linear}
    boundBone {boneName=dyn_head_fin_center_01_cr}
    componentAppliedLinearAccClamp=<3000 3000 3000>
    componentLinearAccScale=<0.1 0.1 0.1>
    componentLinearVelScale=<0.1 0.1 0.1>
    gravityOverride {z=-980}
    physicsBodyDefinitions [{
      boundBone {boneName=dyn_head_fin_center_01_cr}
      boxExtents=<15 25 25>
      constraintSetup {angularLimitsMax=<15 15 15>, angularLimitsMin=<-15 -15 -15>, angularTarget=<1 0 0>}
      localJointOffset=<0.8 0 0>
     }]
    simSpaceSettings {maxAngularAcceleration=35, maxAngularVelocity=7, simSpaceAngularAlpha=1}
   }
   n23 = Anim Dynamics - Bone: dyn_head_fin_center_02_cr!(ComponentPose=@n22, Alpha=1, GravityOverride:Vector="0,0,-980") -> (Pose) {
    angularDampingOverride=1
    angularSpringConstant=10000
    bAngularSpring=true
    bOverrideAngularDamping=true
    bUseGravityOverride=true
    gravityScale=0.35
    simulationSpace=RootRelative
    alphaBoolBlend {blendOption=Linear}
    boundBone {boneName=dyn_head_fin_center_02_cr}
    componentAppliedLinearAccClamp=<3000 3000 3000>
    componentLinearAccScale=<0.1 0.1 0.1>
    componentLinearVelScale=<0.1 0.1 0.1>
    gravityOverride {z=-980}
    physicsBodyDefinitions [{
      boundBone {boneName=dyn_head_fin_center_02_cr}
      boxExtents=<15 25 25>
      constraintSetup {angularLimitsMax=<15 15 8>, angularLimitsMin=<-15 -15 -8>, angularTarget=<1 0 0>}
      localJointOffset=<0.8 0 0>
     }]
    simSpaceSettings {maxAngularAcceleration=35, maxAngularVelocity=7, simSpaceAngularAlpha=1}
   }
   n24 = Anim Dynamics - Bone: dyn_head_fin_back_01_cr!(ComponentPose=@n23, Alpha=1, GravityOverride:Vector="0,0,-980") -> (Pose) {
    angularDampingOverride=1
    angularSpringConstant=10000
    bAngularSpring=true
    bOverrideAngularDamping=true
    bUseGravityOverride=true
    gravityScale=0.35
    simulationSpace=RootRelative
    alphaBoolBlend {blendOption=Linear}
    boundBone {boneName=dyn_head_fin_back_01_cr}
    componentAppliedLinearAccClamp=<3000 3000 3000>
    componentLinearAccScale=<0.1 0.1 0.1>
    componentLinearVelScale=<0.1 0.1 0.1>
    gravityOverride {z=-980}
    physicsBodyDefinitions [{
      boundBone {boneName=dyn_head_fin_back_01_cr}
      boxExtents=<15.5 25.5 25.5>
      constraintSetup {angularLimitsMax=<15 15 15>, angularLimitsMin=<-15 -15 -15>, angularTarget=<1 0 0>}
      localJointOffset=<0.8 0 0>
     }]
    simSpaceSettings {maxAngularAcceleration=35, maxAngularVelocity=7, simSpaceAngularAlpha=1}
   }
   n25 = Anim Dynamics - Bone: dyn_head_fin_back_02_cr!(ComponentPose=@n24, Alpha=1, GravityOverride:Vector="0,0,-980") -> (Pose) {
    angularDampingOverride=1
    angularSpringConstant=10000
    bAngularSpring=true
    bOverrideAngularDamping=true
    bUseGravityOverride=true
    gravityScale=0.35
    simulationSpace=RootRelative
    alphaBoolBlend {blendOption=Linear}
    boundBone {boneName=dyn_head_fin_back_02_cr}
    componentAppliedLinearAccClamp=<3000 3000 3000>
    componentLinearAccScale=<0.1 0.1 0.1>
    componentLinearVelScale=<0.1 0.1 0.1>
    gravityOverride {z=-980}
    physicsBodyDefinitions [{
      boundBone {boneName=dyn_head_fin_back_02_cr}
      boxExtents=<15.5 25.5 25.5>
      constraintSetup {angularLimitsMax=<15 15 8>, angularLimitsMin=<-15 -15 -8>, angularTarget=<1 0 0>}
      localJointOffset=<0.8 0 0>
     }]
    simSpaceSettings {maxAngularAcceleration=35, maxAngularVelocity=7, simSpaceAngularAlpha=1}
   }
   n26 = Anim Dynamics - Bone: dyn_fin_01_rt!(ComponentPose=@n25, Alpha=1, GravityOverride:Vector="0,0,-980") -> (Pose) {
    angularDampingOverride=1
    angularSpringConstant=5000
    bAngularSpring=true
    bOverrideAngularDamping=true
    bUseGravityOverride=true
    gravityScale=0.35
    simulationSpace=RootRelative
    alphaBoolBlend {blendOption=Linear}
    boundBone {boneName=dyn_fin_01_rt}
    componentAppliedLinearAccClamp=<3000 3000 3000>
    componentLinearAccScale=<0.3 0.3 0.3>
    componentLinearVelScale=<0.3 0.3 0.3>
    gravityOverride {z=-980}
    physicsBodyDefinitions [{
      boundBone {boneName=dyn_fin_01_rt}
      boxExtents=<14.5 24.5 24.5>
      constraintSetup {angularLimitsMax=<0 15 0>, angularLimitsMin=<0 0 0>, angularTarget=<1 0 0>}
      localJointOffset=<0.8 0 0>
     }]
    simSpaceSettings {maxAngularAcceleration=35, maxAngularVelocity=7, simSpaceAngularAlpha=1}
   }
   n27 = Anim Dynamics - Bone: dyn_fin_02_rt!(ComponentPose=@n26, Alpha=1, GravityOverride:Vector="0,0,-980") -> (Pose) {
    angularDampingOverride=1
    angularSpringConstant=5000
    bAngularSpring=true
    bOverrideAngularDamping=true
    bUseGravityOverride=true
    gravityScale=0.35
    simulationSpace=RootRelative
    alphaBoolBlend {blendOption=Linear}
    boundBone {boneName=dyn_fin_02_rt}
    componentAppliedLinearAccClamp=<3000 3000 3000>
    componentLinearAccScale=<0.3 0.3 0.3>
    componentLinearVelScale=<0.3 0.3 0.3>
    gravityOverride {z=-980}
    physicsBodyDefinitions [{
      boundBone {boneName=dyn_fin_02_rt}
      boxExtents=<14.5 24.5 24.5>
      constraintSetup {angularLimitsMax=<0 18 0>, angularLimitsMin=<0 0 0>, angularTarget=<1 0 0>}
      localJointOffset=<0.8 0 0>
     }]
    simSpaceSettings {maxAngularAcceleration=35, maxAngularVelocity=7, simSpaceAngularAlpha=1}
   }
   n28 = Anim Dynamics - Bone: dyn_fin_01_lf!(ComponentPose=@n27, Alpha=1, GravityOverride:Vector="0,0,-980") -> (Pose) {
    angularDampingOverride=1
    angularSpringConstant=5000
    bAngularSpring=true
    bOverrideAngularDamping=true
    bUseGravityOverride=true
    gravityScale=0.35
    simulationSpace=RootRelative
    alphaBoolBlend {blendOption=Linear}
    boundBone {boneName=dyn_fin_01_lf}
    componentAppliedLinearAccClamp=<3000 3000 3000>
    componentLinearAccScale=<0.3 0.3 0.3>
    componentLinearVelScale=<0.3 0.3 0.3>
    gravityOverride {z=-980}
    physicsBodyDefinitions [{
      boundBone {boneName=dyn_fin_01_lf}
      boxExtents=<14.5 24.5 24.5>
      constraintSetup {angularLimitsMax=<0 15 0>, angularLimitsMin=<0 0 0>, angularTarget=<1 0 0>}
      localJointOffset=<0.8 0 0>
     }]
    simSpaceSettings {maxAngularAcceleration=35, maxAngularVelocity=7, simSpaceAngularAlpha=1}
   }
   n29 = Anim Dynamics - Bone: dyn_fin_02_lf!(ComponentPose=@n28, Alpha=1, GravityOverride:Vector="0,0,-980") -> (Pose) {
    angularDampingOverride=1
    angularSpringConstant=5000
    bAngularSpring=true
    bOverrideAngularDamping=true
    bUseGravityOverride=true
    gravityScale=0.35
    simulationSpace=RootRelative
    alphaBoolBlend {blendOption=Linear}
    boundBone {boneName=dyn_fin_02_lf}
    componentAppliedLinearAccClamp=<3000 3000 3000>
    componentLinearAccScale=<0.3 0.3 0.3>
    componentLinearVelScale=<0.3 0.3 0.3>
    gravityOverride {z=-980}
    physicsBodyDefinitions [{
      boundBone {boneName=dyn_fin_02_lf}
      boxExtents=<14.5 24.5 24.5>
      constraintSetup {angularLimitsMax=<0 18 0>, angularLimitsMin=<0 0 0>, angularTarget=<1 0 0>}
      localJointOffset=<0.8 0 0>
     }]
    simSpaceSettings {maxAngularAcceleration=35, maxAngularVelocity=7, simSpaceAngularAlpha=1}
   }
   Save cached pose 'fin_cache'(Pose=Component To Local(ComponentPose=@n29)) {cacheName=fin_cache}
   n30 = Get fin() -> (fin:double)
   n31 = "Transform (Modify) Bone - Bone: dyn_fin_01_lf"(
    ComponentPose=Local To Component(LocalPose=Use cached pose 'rigid_cache'())
    Alpha=@n30
    Translation="0,0,0"
    Rotation="25,0,0"
    Scale="1,1,1"
   ) {
    rotationMode=BMM_Additive
    rotationSpace=BCS_BoneSpace
    alphaBoolBlend {blendOption=Linear}
    alphaScaleBiasClamp {bInterpResult=true, interpSpeedDecreasing=15, interpSpeedIncreasing=15}
    boneToModify {boneName=dyn_fin_01_lf}
    rotation {pitch=25}
   }
   n32 = "Transform (Modify) Bone - Bone: dyn_fin_02_lf"(ComponentPose=@n31, Alpha=@n30, Translation="0,0,0", Rotation="5,0,0", Scale="1,1,1") {
    rotationMode=BMM_Additive
    rotationSpace=BCS_BoneSpace
    alphaBoolBlend {blendOption=Linear}
    alphaScaleBiasClamp {bInterpResult=true, interpSpeedDecreasing=15, interpSpeedIncreasing=15}
    boneToModify {boneName=dyn_fin_02_lf}
    rotation {pitch=5}
   }
   n33 = "Transform (Modify) Bone - Bone: dyn_fin_01_rt"(ComponentPose=@n32, Alpha=@n30, Translation="0,0,0", Rotation="-25,0,0", Scale="1,1,1") {
    rotationMode=BMM_Additive
    rotationSpace=BCS_BoneSpace
    alphaBoolBlend {blendOption=Linear}
    alphaScaleBiasClamp {bInterpResult=true, interpSpeedDecreasing=15, interpSpeedIncreasing=15}
    boneToModify {boneName=dyn_fin_01_rt}
    rotation {pitch=-25}
   }
   n34 = "Transform (Modify) Bone - Bone: dyn_fin_01_rt"(ComponentPose=@n33, Alpha=@n30, Translation="0,0,0", Rotation="-5,0,0", Scale="1,1,1") {
    rotationMode=BMM_Additive
    rotationSpace=BCS_BoneSpace
    alphaBoolBlend {blendOption=Linear}
    alphaScaleBiasClamp {bInterpResult=true, interpSpeedDecreasing=15, interpSpeedIncreasing=15}
    boneToModify {boneName=dyn_fin_01_rt}
    rotation {pitch=-5}
   }
   Save cached pose 'modifytransform_cache'(Pose=Component To Local(ComponentPose=@n34)) {cacheName=modifytransform_cache}
  ]}
 EventGraph {nodes [
   Event BlueprintUpdateAnimation()
   > n1 = Sequence!(execute) -> (then_0, then_1, then_2:Exec, then_3:Exec, then_4:Exec, then_5:Exec, then_6:Exec)
   > n2 = Branch(Condition=Get bIsPreviewingItemInFrontEnd() -> (bIsPreviewingItemInFrontEnd:bool))
   > Set RbanFEAlpha(RbanFEAlpha:double=0) -> (Output_Get:double)
   > Set RbanAlpha(RbanAlpha:double=1) -> (Output_Get:double)
   Set RbanFEAlpha(execute=@n2.else, RbanFEAlpha=1)
   > Set RbanAlpha(RbanAlpha=0)
   n3 = GetPartSkeletalMeshComponent()
   Is Valid(exec=@n1.then_1, InputObject=@n3)
   n4 = GetAttachParent!(self:Scene Component Object Reference=@n3) -> (ReturnValue:Scene Component Object Reference)
   > Is Valid(InputObject=@n4)
   > Set Body_Mesh(Body_Mesh:Scene Component Object Reference=@n4) -> (Output_Get:Scene Component Object Reference)
   > n5 = Sequence()
   n6 = Get Body_Mesh() -> (Body_Mesh:Scene Component Object Reference)
   n7 = BreakTransform(InTransform=GetSocketTransform(self=@n6, InSocketName=root, TransformSpace=RTS_World))
   Set root_location(execute=@n5.then_6, root_location:Vector=@n7) -> (Output_Get:Vector)
   > Set root_space_rotation(root_space_rotation:Rotator=@n7.Rotation) -> (Output_Get:Rotator)
   n8 = BreakTransform(InTransform=GetSocketTransform(self=@n6, InSocketName=pelvis, TransformSpace=RTS_World))
   Set pelvis_location(execute=@n5.then_0, pelvis_location:Vector=@n8) -> (Output_Get:Vector)
   > Set pelvis_rotation(pelvis_rotation:Rotator=@n8.Rotation) -> (Output_Get:Rotator)
   n9 = BreakTransform(InTransform=GetSocketTransform(self=@n6, InSocketName=thigh_l, TransformSpace=RTS_World))
   Set thigh_left_location(execute=@n5.then_1, thigh_left_location:Vector=@n9) -> (Output_Get:Vector)
   > Set thigh_left_rotation(thigh_left_rotation:Rotator=@n9.Rotation) -> (Output_Get:Rotator)
   n10 = BreakTransform(InTransform=GetSocketTransform(self=@n6, InSocketName=thigh_r, TransformSpace=RTS_World))
   Set thigh_right_location(execute=@n5.then_2, thigh_right_location:Vector=@n10) -> (Output_Get:Vector)
   > Set thigh_right_rotation(thigh_right_rotation:Rotator=@n10.Rotation) -> (Output_Get:Rotator)
   n11 = BreakTransform(InTransform=GetSocketTransform(self=@n6, InSocketName=calf_r, TransformSpace=RTS_World))
   Set calf_right_location(execute=@n5.then_4, calf_right_location:Vector=@n11) -> (Output_Get:Vector)
   > Set calf_right_rotation(calf_right_rotation:Rotator=@n11.Rotation) -> (Output_Get:Rotator)
   n12 = BreakTransform(InTransform=GetSocketTransform(self=@n6, InSocketName=calf_l, TransformSpace=RTS_World))
   Set calf_left_location(execute=@n5.then_3, calf_left_location:Vector=@n12) -> (Output_Get:Vector)
   > Set calf_left_rotation(calf_left_rotation:Rotator=@n12.Rotation) -> (Output_Get:Rotator)
   n13 = BreakTransform(InTransform=GetSocketTransform(self=@n6, InSocketName=root, TransformSpace=RTS_World))
   Set body_root_location(execute=@n5.then_5, body_root_location:Vector=@n13) -> (Output_Get:Vector)
   > Set body_root_rotation(body_root_rotation:Rotator=@n13.Rotation) -> (Output_Get:Rotator)
   n14 = Branch(execute=@n1.then_4, Condition=Get bIsPlayingMeleeAnim())
   > Set SimSpaceSettings(SimSpaceSettings:Sim Space Settings Structure=Get SimSpaceSW() -> (SimSpaceSW:Sim Space Settings Structure)) -> (Output_Get:Sim Space Settings Structure)
   Set SimSpaceSettings(execute=@n14.else, SimSpaceSettings=Get SimSpacePIE() -> (SimSpacePIE:Sim Space Settings Structure))
   Branch(execute=@n1.then_5, Condition=NOT Boolean(A=Get bIsInFrontEnd()))
   > Set RbanFEAlpha(RbanFEAlpha=0)
   > Set RbanAlpha(RbanAlpha=1)
   Delay(execute=@n1.then_6, Duration=RandomFloatInRange(Min=RandomFloatInRange(Min=0.1, Max=7), Max=10))
   > Set Blink(Blink:double=1) -> (Output_Get:double)
   > Delay(Duration=0.2)
   > Set Blink(Blink=0)
   > Delay(Duration=0.1)
   > Set Blink(Blink=0)
   Delay(execute=@n1.then_3, Duration=RandomFloatInRange(Min=RandomFloatInRange(Min=0.1, Max=1), Max=3))
   > Set fin(fin:double=1) -> (Output_Get:double)
   > Delay(Duration=RandomFloatInRange(Min=RandomFloatInRange(Min=0.01, Max=0.3), Max=0.5))
   > Set fin(fin=0)
   > Delay(Duration=0.1)
   > Set fin(fin=0)
   Delay(execute=@n1.then_2, Duration=RandomFloatInRange(Min=3, Max=6))
   > Set mouth_open(mouth_open:double=1) -> (Output_Get:double)
   > Delay(Duration=RandomFloatInRange(Min=0.5, Max=0.8))
   > Set mouth_open(mouth_open=0)
   Event OnCharacterCustomizationLoaded()
   > n15 = Branch(Condition=Get bIsInFrontEnd())
   > Set Pickaxe_Rot(Pickaxe_Rot:double=1) -> (Output_Get:double)
   Set Pickaxe_Rot(execute=@n15.else, Pickaxe_Rot=0)
   TryGetPawnOwner()
   RandomFloatInRange(Min=0, Max=0.5)
   RandomFloatInRange(Min=0.01, Max=1)
  ]}
]
variables [
 Blink {type=number}
 Body_Mesh {type=SceneComponent, default=null}
 body_root_location {type=Vector}
 body_root_rotation {type=Rotator}
 calf_left_location {type=Vector}
 calf_left_rotation {type=Rotator}
 calf_right_location {type=Vector}
 calf_right_rotation {type=Rotator}
 fin {type=number}
 Gravity_pickaxe {type=Vector}
 GravityOverride {type=Vector}
 mouth_open {type=number}
 pelvis_location {type=Vector}
 pelvis_rotation {type=Rotator}
 Pickaxe_Rot {type=number}
 RBAN Settings Out {type=FortRigidBodyAnimNodeParameters}
 RbanAlpha {type=number}
 RbanFEAlpha {type=number}
 root_location {type=Vector}
 root_space_rotation {type=Rotator}
 SimSpacePIE {
  type=SimSpaceSettings
  default {
   dampingAlpha=0
   maxAngularAcceleration=45
   maxAngularVelocity=8
   maxLinearAcceleration=3500
   maxLinearVelocity=700
   velocityScaleZ=0.5
   worldAlpha=0.8
   externalLinearDragV=<0.5 0.5 0.35>
  }
 }
 SimSpaceSettings {
  type=SimSpaceSettings
  default {
   dampingAlpha=1
   maxAngularAcceleration=10000
   maxAngularVelocity=10000
   maxLinearAcceleration=10000
   maxLinearVelocity=10000
   velocityScaleZ=1
   worldAlpha=0
   externalLinearDragV=<0 0 0>
  }
 }
 SimSpaceSW {
  type=SimSpaceSettings
  default {
   dampingAlpha=0
   maxAngularAcceleration=60
   maxAngularVelocity=15
   maxLinearAcceleration=3500
   maxLinearVelocity=700
   velocityScaleZ=0.5
   worldAlpha=1
   externalLinearDragV=<0 0 0>
  }
 }
 thigh_left_location {type=Vector}
 thigh_left_rotation {type=Rotator}
 thigh_right_location {type=Vector}
 thigh_right_rotation {type=Rotator}
]
